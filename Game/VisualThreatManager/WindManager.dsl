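; Wind manager Blueprint that visualizes gameplay wind impulse cylinders with storm wind shader meshes.
; Creation events add or update a mesh matched by XY location, animating its MID radii for wind changes; death events destroy it.
; Delta completion copies animated magnitude to static and clears the animating flag.
; It also offsets wind particles from view at BeginPlay and holds test wind events, empty event stubs and a texel-snapped capture-camera helper.
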(event Custom|Play%20Water%20Splash%20Particle%20System%20At%20Location
  (params (Particle%20Transform :type Transform)))

(event Custom|Register%20player%20for%20render%20to%20texture%20purposes
  (params (Effect%20Water%20Interaction%20FX :type Primitive%20Component%20Object%20Reference) (Gameplay%20Wind%20INteraction%20Effects :type Primitive%20Component%20Object%20Reference)))

(event Custom|Add%20Render%20To%20Texture%20Particle
  (params (EmitterTemplate :type Particle%20System%20Object%20Reference) (InTransform :type Transform%20%28by%20ref%29)))

(event Wind|EventAddWindParticleSystemComponent
  (params (ParticleSystemComponent :type Cascade%20Particle%20System%20Component%20Object%20Reference)))

(event Custom|Add%20Wind%20Component
  (params (InComponent :type Primitive%20Component%20Object%20Reference)))

(event EventBeginPlay
  ;[1 0 0] Hides the wind force particle effects from the main view
  (Rendering|Material|SetScalarParameterValue "/Game/Packages/Fortress_SharedMaterials/GlobalMaterialParameters/FortniteMaterialParameters.FortniteMaterialParameters" "WindVectorParticleOffset" -2e+04 :!class /Script/Engine.KismetMaterialLibrary))

(event Custom|Update%20Test
  (params (NewMagnitude :type Float%20%28double-precision%29))
  ;[0.4773 0.4503 1] Gamplay wind
  ;;[1 1 1] Debug Wind
  (Wind|SetWindImpulseCylinderMagnitude (Variables|Default|GetNewVar_1) NewMagnitude 10))

(event Custom|SpawnTestWind
  (params (Radius :type Float%20%28double-precision%29 :default 4096) (ImpulseMagnitude :type Float%20%28double-precision%29))
  ;[0.4773 0.4503 1] Gamplay wind
  ;;[1 1 1] Debug Wind
  (bind _returnvalue (Wind|AddWindImpulseCylinder (Wind|MakeWindImpulseCylinder (Math|Vector|MakeVector -1704.0 16236.0 10) Radius (+ Radius 1024) ImpulseMagnitude (Math|Vector|MakeVector 0.0)) 10))
  (Variables|Default|SetNewVar_1 _returnvalue))

(event Wind|EventOnWindImpulseCylinderCreation
  (params (WindImpulseCylinder :type Fort%20Wind%20Impulse%20Cylinder%20Structure%20%28by%20ref%29) (WindImpulseCylinderDelta :type Fort%20Wind%20Impulse%20Cylinder%20Delta%20Structure%20%28by%20ref%29))
  ;[0.4773 0.4503 1] Gamplay wind
  (CallFunction|UpdateStormWindCylinder WindImpulseCylinder WindImpulseCylinderDelta))

(event Wind|EventOnWindImpulseCylinderDeath
  (params (WindImpulseCylinder :type Fort%20Wind%20Impulse%20Cylinder%20Structure%20%28by%20ref%29))
  ;[0.4773 0.4503 1] Gamplay wind
  (CallFunction|RemoveWindCylinder WindImpulseCylinder))

(event Wind|EventOnWindImpulseCylinderDeltaComplete
  (params (WindImpulseCylinder :type Fort%20Wind%20Impulse%20Cylinder%20Structure%20%28by%20ref%29))
  (bind _stormcylindermeshes (Variables|Default|GetStormCylinderMeshes))
  ;[0.4773 0.4503 1] Gamplay wind
  ;;[1 1 1] Find shader associated and flip animation values
  (bind _windlocation (Wind|BreakWindImpulseCylinderAdvanced WindImpulseCylinder))
  (bind (_storm_cyl_mesh_match _wind_cyl_mesh_array_match_index) (CallFunction|FindMatchingWindMeshIndexandWriteBool _windlocation))
  (if _storm_cyl_mesh_match
    (bind _returnvalue (Rendering|Material|GetScalarParameterValue (Utilities|Casting|CastToMaterialInstanceDynamic (Rendering|Material|GetMaterial (Utilities|Array|Get(acopy) _stormcylindermeshes _wind_cyl_mesh_array_match_index) 0)) (Utilities|Name|MakeLiteralName "Animated Wind Magnitude") :!class /Script/Engine.MaterialInstanceDynamic))
    (Rendering|Material|SetScalarParameterValue (Utilities|Casting|CastToMaterialInstanceDynamic (Rendering|Material|GetMaterial (Utilities|Array|Get(acopy) _stormcylindermeshes _wind_cyl_mesh_array_match_index) 0)) (Utilities|Name|MakeLiteralName "Static Wind Magnitude") _returnvalue :!class /Script/Engine.MaterialInstanceDynamic)
    (Rendering|Material|SetScalarParameterValue (Utilities|Casting|CastToMaterialInstanceDynamic (Rendering|Material|GetMaterial (Utilities|Array|Get(acopy) _stormcylindermeshes _wind_cyl_mesh_array_match_index) 0)) (Utilities|Name|MakeLiteralName "Wind Still 0 or Animating 1") 0.0 :!class /Script/Engine.MaterialInstanceDynamic)))

(fn CalculateCameraPosition
  (params (Scale :type Float%20%28double-precision%29) (Scene%20Capture :type Scene%20Capture%20Component%202D%20Object%20Reference))
  (return (* (Math|Vector|MakeVector (Math|Conversions|ToFloat(Integer) (Math|Float|Ceil (.x (- (* (/ (Camera|GetCameraLocation (Game|GetPlayerCameraManager 0)) Scale) (Math|Vector|MakeVector (Math|Conversions|ToFloat(Integer) (Variables|/Script/Engine.TextureRenderTarget2D|GetSizeX (Variables|/Script/Engine.SceneCaptureComponent2D|GetTextureTarget Scene%20Capture))) (Math|Conversions|ToFloat(Integer) (Variables|/Script/Engine.TextureRenderTarget2D|GetSizeY (Variables|/Script/Engine.SceneCaptureComponent2D|GetTextureTarget Scene%20Capture))) 1)) 0.5)))) (Math|Conversions|ToFloat(Integer) (Math|Float|Ceil (.y (- (* (/ (Camera|GetCameraLocation (Game|GetPlayerCameraManager 0)) Scale) (Math|Vector|MakeVector (Math|Conversions|ToFloat(Integer) (Variables|/Script/Engine.TextureRenderTarget2D|GetSizeX (Variables|/Script/Engine.SceneCaptureComponent2D|GetTextureTarget Scene%20Capture))) (Math|Conversions|ToFloat(Integer) (Variables|/Script/Engine.TextureRenderTarget2D|GetSizeY (Variables|/Script/Engine.SceneCaptureComponent2D|GetTextureTarget Scene%20Capture))) 1)) 0.5)))) (Math|Conversions|ToFloat(Integer) (Math|Float|Ceil (.z (- (* (/ (Camera|GetCameraLocation (Game|GetPlayerCameraManager 0)) Scale) (Math|Vector|MakeVector (Math|Conversions|ToFloat(Integer) (Variables|/Script/Engine.TextureRenderTarget2D|GetSizeX (Variables|/Script/Engine.SceneCaptureComponent2D|GetTextureTarget Scene%20Capture))) (Math|Conversions|ToFloat(Integer) (Variables|/Script/Engine.TextureRenderTarget2D|GetSizeY (Variables|/Script/Engine.SceneCaptureComponent2D|GetTextureTarget Scene%20Capture))) 1)) 0.5))))) (/ (Math|Vector|MakeVector Scale Scale Scale) (Math|Vector|MakeVector (Math|Conversions|ToFloat(Integer) (Variables|/Script/Engine.TextureRenderTarget2D|GetSizeX (Variables|/Script/Engine.SceneCaptureComponent2D|GetTextureTarget Scene%20Capture))) (Math|Conversions|ToFloat(Integer) (Variables|/Script/Engine.TextureRenderTarget2D|GetSizeY (Variables|/Script/Engine.SceneCaptureComponent2D|GetTextureTarget Scene%20Capture))) 1)))))

(fn FindMatchingWindMeshIndexAndWriteBool
  (params (Wind%20location :type Vector))
  (Variables|Default|SetStorm%20Cyl%20Mesh%20Match false)
  (Variables|Default|SetWind%20Cyl%20Mesh%20Array%20Match%20Index -1)
  (bind (_array_element _array_index) (Utilities|Array|ForEachLoopwithBreak (Variables|Default|GetStormCylinderMeshes))
    (:LoopBody
      (bind _output_get (Variables|Default|SetStorm%20Cyl%20Mesh%20Match (Math|Vector|Equal(Vector) (* (Transformation|GetWorldLocation _array_element) "1.000000,1.000000, 0") (* Wind%20location "1.000000,1.000000, 0") 1)))
      (if _output_get
        (Variables|Default|SetWind%20Cyl%20Mesh%20Array%20Match%20Index _array_index)))
    (:Completed
      _))
  (return (Variables|Default|GetStorm%20Cyl%20Mesh%20Match) (Variables|Default|GetWind%20Cyl%20Mesh%20Array%20Match%20Index)))

(fn RemoveWindCylinder
  (params (WindImpulseCylinder :type Fort%20Wind%20Impulse%20Cylinder%20Structure%20%28by%20ref%29))
  (bind _stormcylindermeshes (Variables|Default|GetStormCylinderMeshes))
  (bind _windlocation (Wind|BreakWindImpulseCylinderAdvanced WindImpulseCylinder))
  (bind _array_element (Utilities|Array|ForEachLoopwithBreak _stormcylindermeshes)
    (:LoopBody
      (bind _output_get (LocalVariables|SetCurrentStormCylinder _array_element))
      (if (Math|Vector|Equal(Vector) (* (Transformation|GetWorldLocation _output_get) "1.000000,1.000000, 0") (* _windlocation "1.000000,1.000000, 0") 0.0001)
        (Utilities|Array|RemoveItem _stormcylindermeshes _output_get)
        (Components|DestroyComponent _output_get)))))

(fn SetWindMatVariables
  (params (MID :type Material%20Instance%20Dynamic%20Object%20Reference) (Set%20AnimatedCylinderCenter :type Boolean) (AnimatedCylinderCenter :type Vector) (Set%20Animating%20Wind%20Magnitude :type Boolean) (Animating%20Wind%20Magnitude :type Float%20%28double-precision%29) (Set%20Static%20Wind%20Magnitude :type Boolean) (Static%20Wind%20Magnitude :type Float%20%28double-precision%29) (Set%20CylinderCenter :type Boolean) (CylinderCenter :type Vector) (Set%20WindVector :type Boolean) (WindVector :type Vector) (Set%20WindCrossVector :type Boolean) (WindCrossVector :type Vector) (Set%20OverallOuterRadius :type Boolean) (OverallOuterRadius :type Float%20%28double-precision%29) (Set%20OverallInnerRadius :type Boolean) (OverallInnerRadius :type Float%20%28double-precision%29) (Set%20Wind%20Still%200%20or%20Animating%201 :type Boolean) (Wind%20Still%200%20or%20Animating%201 :type Float%20%28double-precision%29) (Set%20Circular%20Shader? :type Boolean) (Circular%20Shader? :type Float%20%28double-precision%29))
  (if (Variables|GetSet%20AnimatedCylinderCenter)
    (Rendering|Material|SetVectorParameterValue MID (Utilities|Name|MakeLiteralName "AnimatedCylinderCenter") (Math|Conversions|ToLinearColor(Vector) (Variables|GetAnimatedCylinderCenter)) :!class /Script/Engine.MaterialInstanceDynamic))
  (if (Variables|GetSet%20Animating%20Wind%20Magnitude)
    (Rendering|Material|SetScalarParameterValue MID (Utilities|Name|MakeLiteralName "Animated Wind Magnitude") (Variables|GetAnimating%20Wind%20Magnitude) :!class /Script/Engine.MaterialInstanceDynamic))
  (if (Variables|GetSet%20Static%20Wind%20Magnitude)
    (Rendering|Material|SetScalarParameterValue MID (Utilities|Name|MakeLiteralName "Static Wind Magnitude") (Variables|GetStatic%20Wind%20Magnitude) :!class /Script/Engine.MaterialInstanceDynamic))
  (if (Variables|GetSet%20CylinderCenter)
    (Rendering|Material|SetVectorParameterValue MID (Utilities|Name|MakeLiteralName "CylinderCenter") (Math|Conversions|ToLinearColor(Vector) (Variables|GetCylinderCenter)) :!class /Script/Engine.MaterialInstanceDynamic))
  (if (Variables|GetSet%20WindVector)
    (Rendering|Material|SetVectorParameterValue MID (Utilities|Name|MakeLiteralName "WindVector") (Math|Conversions|ToLinearColor(Vector) (Variables|GetWindVector)) :!class /Script/Engine.MaterialInstanceDynamic))
  (if (Variables|GetSet%20WindCrossVector)
    (Rendering|Material|SetVectorParameterValue MID (Utilities|Name|MakeLiteralName "WindCrossVector") (Math|Conversions|ToLinearColor(Vector) (Variables|GetWindCrossVector)) :!class /Script/Engine.MaterialInstanceDynamic))
  (if (Variables|GetSet%20OverallOuterRadius)
    (Rendering|Material|SetScalarParameterValue MID (Utilities|Name|MakeLiteralName "OverallOuterRadius") (Variables|GetOverallOuterRadius) :!class /Script/Engine.MaterialInstanceDynamic))
  (if (Variables|GetSet%20OverallInnerRadius)
    (Rendering|Material|SetScalarParameterValue MID (Utilities|Name|MakeLiteralName "OverallInnerRadius") (Variables|GetOverallInnerRadius) :!class /Script/Engine.MaterialInstanceDynamic))
  (if (Variables|GetSet%20Wind%20Still%200%20or%20Animating%201)
    (Rendering|Material|SetScalarParameterValue MID (Utilities|Name|MakeLiteralName "Wind Still 0 or Animating 1") (Variables|GetWind%20Still%200%20or%20Animating%201) :!class /Script/Engine.MaterialInstanceDynamic))
  (if (Variables|GetSet%20Circular%20Shader?)
    (Rendering|Material|SetScalarParameterValue MID (Utilities|Name|MakeLiteralName "Circular Shader?") (Variables|GetCircular%20Shader?) :!class /Script/Engine.MaterialInstanceDynamic))
  (return false))

(fn UpdateStormWindCylinder
  (params (Wind%20Impulse :type Fort%20Wind%20Impulse%20Cylinder%20Structure) (Wind%20Delta :type Fort%20Wind%20Impulse%20Cylinder%20Delta%20Structure))
  ;[0 0 0] non-pure break nodes made this process fairly messy
  (bind (_winddeltacenter _bwindimpulseinitialized _bwindripplesoutward _winddeltasectionwidth _winddeltainnersectionradius _winddeltaoutersectionradius _winddeltamaximumradius _winddeltadesiredoverallblendtime _winddeltasectionblendtime _winddeltasectioncurrentblendtime _winddeltapreviousmagnitude _winddeltasectioncurrentmagnitude _winddeltadesiredmagnitude _winddeltaouterworldbounds _winddeltainnerworldbounds _windimpulsebounds _windimpulsehandletomodify) (Wind|BreakWindImpulseCylinderDelta Wind%20Delta))
  (bind (_windlocation _windinnerradius _windouterradius _windworlddirection) (Wind|BreakWindImpulseCylinderAdvanced Wind%20Impulse))
  ;[0 0 0] Find Existing Cloud Wind
  (CallFunction|FindMatchingWindMeshIndexandWriteBool _windlocation)
  (if (Variables|Default|GetStorm%20Cyl%20Mesh%20Match)
    ;[0 0 0] read and update parameters
    (bind _asmaterial_instance_dynamic (Utilities|Casting|CastToMaterialInstanceDynamic (Rendering|Material|GetMaterial (Utilities|Array|Get(acopy) (Variables|Default|GetStormCylinderMeshes) (Variables|Default|GetWind%20Cyl%20Mesh%20Array%20Match%20Index)) 0)))
    (CallFunction|SetWindMatVariables _asmaterial_instance_dynamic false "0, 0, 0" true _winddeltadesiredmagnitude false 0.0 false "0, 0, 0" false "0, 0, 0" false "0, 0, 0" false 0.0 false 0.0 true 1)
    (CallFunction|UpdateWindDeltaCyl (* _windouterradius 2) _winddeltadesiredoverallblendtime _asmaterial_instance_dynamic (Math|Vector|Equal(Vector) _windworlddirection "0, 0, 0" 0.0001))
    (else
      ;[0 0 0] make new mesh
      (bind _returnvalue (AddComponent|NODE_AddStaticMeshComponent-1 :RelativeTransform (Math|Transform|MakeTransform "0, 0, 0")))
      (Utilities|Array|Add (Variables|Default|GetStormCylinderMeshes) _returnvalue)
      (Transformation|SetWorldTransform _returnvalue (Math|Transform|MakeTransform _windlocation (Math|Conversions|RotationFromXVector (Math|Vector|CrossProduct "0, 0,1.000000" _windworlddirection)) (Math|Conversions|ToVector(Float) (* _windouterradius 2))) false true)
      (bind _returnvalue_2 (Rendering|Material|CreateDynamicMaterialInstance _returnvalue 0 "/Game/WaterAndWind/Materials/WindForceCylinder/StormWindShader_Inst.StormWindShader_Inst" "None" :!class /Script/Engine.PrimitiveComponent))
      (CallFunction|SetWindMatVariables _returnvalue_2 true _winddeltacenter true _winddeltadesiredmagnitude true 0.0 true _windlocation true (Math|Vector|CrossProduct "0, 0,1.000000" _windworlddirection) true _windworlddirection true _windouterradius true _windinnerradius true 1 true (Math|Conversions|ToFloat(Boolean) (Math|Vector|Equal(Vector) _windworlddirection "0, 0, 0" 0.0001)))
      (CallFunction|UpdateWindDeltaCyl (* _windouterradius 2) _winddeltadesiredoverallblendtime _returnvalue_2 (Math|Vector|Equal(Vector) _windworlddirection "0, 0, 0" 0.0001)))))

(fn UpdateWindDeltaCyl
  (params (Static%20Cylinder%20Outer%20Radius :type Float%20%28double-precision%29) (Duration :type Float%20%28double-precision%29) (MID :type Material%20Instance%20Dynamic%20Object%20Reference) (Is%20Wind%20Circular? :type Boolean))
  (bind _delta_wind_falloff_in_world_units (Variables|Default|GetDelta%20Wind%20Falloff%20in%20World%20Units))
  (FortFX|AddMIDAnimationFloat MID "AnimatedInnerRadius" (select (Variables|GetIs%20Wind%20Circular?) (Variables|GetStatic%20Cylinder%20Outer%20Radius) (* -1 _delta_wind_falloff_in_world_units)) (select (Variables|GetIs%20Wind%20Circular?) (* -1 _delta_wind_falloff_in_world_units) (Variables|GetStatic%20Cylinder%20Outer%20Radius)) 0 false (Variables|GetDuration))
  (FortFX|AddMIDAnimationFloat MID "AnimatedOuterRadius" (select (Variables|GetIs%20Wind%20Circular?) (+ _delta_wind_falloff_in_world_units (Variables|GetStatic%20Cylinder%20Outer%20Radius)) 0.0) (select (Variables|GetIs%20Wind%20Circular?) 0.0 (+ _delta_wind_falloff_in_world_units (Variables|GetStatic%20Cylinder%20Outer%20Radius))) 0 false (Variables|GetDuration)))

(fn ConstructionScript
  (|CallParentFunction)
  (AddComponent|NODE_AddParticleSystemComponent-4))
